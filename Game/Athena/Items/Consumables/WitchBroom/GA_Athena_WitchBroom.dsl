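; Witch Broom consumable gameplay ability that extends a parent leap ability for the player pawn.
; On activation it ejects the player from grind rails (blocking re-grind, caching velocity, switching to walking) or ziplines, as borrowed from Katana.
; The server sends a "rode witch broom" quest event, then the ability ends.
; On leap finish, only if HolsterWeapons is true, it grants a till-landing holster ability and removes the grind-block tag.
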(event EventOnAbilityActivated
  (|CallParentFunction)
  ;[1 1 1] Checks if player is Ziplining or Grinding to eject first - behavior borrowed from Katana
  (bind _outfortplayerpawn (CallFunction|GetAvatarasValidFortPlayerPawn)
    (:ValidFortPlayerPawn
      (Variables|Default|SetFortPlayerPawn _outfortplayerpawn)
      (bind _fortplayerpawn (Variables|Default|GetFortPlayerPawn)
        (:then
          (Networking|ForceNetUpdate _fortplayerpawn)
          (if (GameplayTags|HasMatchingGameplayTag (Variables|Default|GetFortPlayerPawn) "(TagName=\"MovementMode.Grinding\")")
            (Ability|GameplayEffect|AddLooseGameplayTags (Variables|Default|GetFortPlayerPawn) (Variables|Default|GetBlockGrindRailsTag))
            (Variables|Default|SetCachedVelocity (Pawn|Components|CharacterMovement|GetLastUpdateVelocity (Variables|/Script/Engine.Character|GetCharacterMovement (Variables|Default|GetFortPlayerPawn))))
            (Pawn|Components|CharacterMovement|SetMovementMode (Variables|/Script/Engine.Character|GetCharacterMovement (Variables|Default|GetFortPlayerPawn)) "MOVE_Walking" 0)
            (else
              (if (GameplayTags|HasMatchingGameplayTag (Variables|Default|GetFortPlayerPawn) "(TagName=\"MovementMode.ZipliningOnSpline\")")
                (Zipline|EndZiplining (Variables|Default|GetFortPlayerPawn) false false))))
          ;[1 1 1] Quest Stuff
          (if (Networking|IsServer)
            (bind _asfort_player_pawn (Utilities|Casting|CastToFortPlayerPawn (Ability|GetAvatarActorfromActorInfo)))
            (if (Utilities|IsValid (Fortnite|GetFortPlayerControllerfromActor _asfort_player_pawn) :!class /Script/Engine.KismetSystemLibrary)
              (Quest|SendComplexCustomStatEvent (Quests|GetQuestManager (Fortnite|GetFortPlayerControllerfromActor _asfort_player_pawn) "Invalid") 0 (Variables|Default|GetQuestTag) "(GameplayTags=)" 1) ; quest event: "rode witch broom"
            )))
        (:else
          _)))
    (:AvatarCastFailed
      _))
  (Ability|EndAbility))

(event EventOnLeapFinish
  ;[1 1 1] If HoslterWeapons is True, player cannot pull out weapons until landing. Applied GA handles the holstering
  (|CallParentFunction)
  (if (ScalableFloat|GetValueasBool (Variables|Default|GetHolsterWeapons))
    (bind _asfort_player_pawn (Utilities|Casting|CastToFortPlayerPawn (Ability|GetAvatarActorfromActorInfo)))
    (Ability|Advanced|GrantandActivateAbilityonSelf (Ability|GetAbilitySystemComponent _asfort_player_pawn) "/Game/Athena/Items/Consumables/WitchBroom/Holster/GA_TillLandHolster_Witchbroom.GA_TillLandHolster_Witchbroom_C")
    ;[1 1 1] Removes tag the blocks grinding so players can grind again
    (Ability|GameplayEffect|RemoveLooseGameplayTags (Variables|Default|GetFortPlayerPawn) (Variables|Default|GetBlockGrindRailsTag))))
