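; Gameplay ability that cancels the Witch Broom glide in Athena.
; On activation it reads the CanCancelGlide scalable float; when false, the player cannot cancel and rides the glide all the way down.
; When allowed, it runs the parent cancel logic, then gets the avatar's Fort player pawn and pops the broom glider override so it no longer serves as the glider.
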(event Ability|EventActivateAbility
  ;[1 1 1] If CanCancelGlide is False, player cannot cancel glide and has to go all the way down
  (if (ScalableFloat|GetValueasBool (Variables|Default|GetCanCancelGlide))
    (|CallParentFunction)
    ;[1 1 1] Removes the witchbroom as a glider after cancelling the ability
    (bind _outfortplayerpawn (CallFunction|GetAvatarasValidFortPlayerPawn)
      (:ValidFortPlayerPawn
        (Parachute|PopGliderOverride _outfortplayerpawn (Variables|Default|GetGlider))))))
