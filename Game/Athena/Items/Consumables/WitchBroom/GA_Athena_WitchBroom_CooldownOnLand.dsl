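; Gameplay ability for the Witch Broom consumable that removes a glider override when the player enters a wind spline.
; On activation it caches the avatar as a valid Fort player pawn, then waits for the InWindSpline wind status tag to be added.
; When that tag arrives, it pops the glider held in a variable from the pawn.
; Despite the CooldownOnLand name, this graph contains no cooldown or landing logic.
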(event Ability|EventActivateAbility
  (|CallParentFunction)
  (bind _outfortplayerpawn (CallFunction|GetAvatarasValidFortPlayerPawn)
    (:ValidFortPlayerPawn
      (Variables|Default|SetFortPlayerPawn _outfortplayerpawn)
      (Ability|Tasks|WaitGameplayTagAdd "(TagName=\"Athena.Gameplay.Status.Wind.InWindSpline\")"
        (:Added
          (Parachute|PopGliderOverride (Variables|Default|GetFortPlayerPawn) (Variables|Default|GetIn%20Glider%20to%20Remove)))))))

(event Ability|EventOnEndAbility
  (params (bWasCancelled :type Boolean)))
